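# Niagara one-shot impact effect for the CiderPetalClover emote's return, pairing a dust puff with a shockwave.
# Dust bursts sprites from a small ring, pushed along the shape tangent and slightly upward, slowed by drag and gravity.
# Curve-driven dynamic material parameters, size, rotation and alpha animate each puff over its life.
# Shockwave is a short-lived mesh that expands in XY and fades, with opacity set by time of day.
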NiagaraSystem /BRCosmetics/Animation/Game/MainPlayer/Emotes/CiderPetalClover/FX/NS_Emote_CiderPetalClover_ReturnImpactDust
dependencies {
 usedDataInterfaces [&NiagaraDataInterfaceCurve, &NiagaraDataInterfaceVectorCurve]
 usedDynamicInputs [/Game/Effects/Niagara/DynamicInputs/Math/FN_NormalizedParticleCount]
 usedModules [&SystemState]
}
emitters {
 Dust {
  emitterUpdateScript [
   &EmitterState()
   &SpawnBurst_Instantaneous(Spawn Count=14, Use Spawn Probability=false, UseLoopCountLimit=false)
   /SetVariables(Emitter.SizeRad=25) {isSetParameters=true}
  ]
  particleSpawnScript [
   /SetVariables(Particles.DustColor:LinearColor=&RandomRangeLinearColor(Minimum=<rgba 0.4635 0.4212 0.4025 1>, Maximum=<rgba 0.3229 0.2978 0.2778 1>)) {isSetParameters=true}
   /SetVariables(Particles.DustColor:LinearColor=/Niagara/DynamicInputs/LinearColor/MakeLinearColorFromVector4(Vector:Vector4f=&MakeVector4(
      X=&MakeFloatFromLinearColor(LinearColor=$Particles.DustColor, Channel:ENiagaraFloatFromLinearColorOptions=::NewEnumerator0)
      Y=&MakeFloatFromLinearColor(LinearColor=$Particles.DustColor, Channel:ENiagaraFloatFromLinearColorOptions=::NewEnumerator1)
      Z=&MakeFloatFromLinearColor(LinearColor=$Particles.DustColor, Channel:ENiagaraFloatFromLinearColorOptions=::NewEnumerator2)
      W=&Multiply_Float(
       A=&MakeFloatFromLinearColor(LinearColor=$Particles.DustColor, Channel:ENiagaraFloatFromLinearColorOptions=::NewEnumerator3)
       B=&FloatFromCurve(
        FloatCurve=DataInterface() {propertyValues {Curve {
           defaultValue=1
           keys [{interpMode=RCIM_Cubic, tangentMode=RCTM_User, time=0, value=0.3}, {interpMode=RCIM_Cubic, tangentMode=RCTM_User, time=1, value=0.55}]
          }}}
        CurveIndex=$Emitter.NormalizedLoopAge
       )
      )
     ))) {isSetParameters=true}
   /SetVariables(
    Particles.SizeMax:NiagaraFloat=&MaskFloatBySpawnGroup(Masked Float=60, Passthrough Float=20)
    Particles.SizeMin:NiagaraFloat=&MaskFloatBySpawnGroup(Masked Float=85, Passthrough Float=40)
   ) {isSetParameters=true}
   /SetVariables(Particles.SizeCurve:NiagaraFloat=&Lerp_Float(
     A=$Particles.SizeMin
     B=$Particles.SizeMax
     Alpha=&FloatFromCurve(
      FloatCurve=DataInterface() {propertyValues {Curve {keys [
          {interpMode=RCIM_Cubic, tangentMode=RCTM_Break, time=0, value=0}
          {arriveTangentWeight=0.9217, interpMode=RCIM_Cubic, tangentMode=RCTM_Break, time=1, value=1}
         ]}}}
      CurveIndex=&RandomRangeFloat(Minimum=0, Maximum=1)
     )
    )) {isSetParameters=true}
   /SetVariables(Particles.InitSpriteRotRate=&UniformRangedFloat(Minimum=-1, Maximum=1)) {isSetParameters=true}
   &InitializeParticle(
    Color=$Particles.DustColor
    Lifetime Max=1.6
    Lifetime Min=1.2
    Lifetime Mode=::NewEnumerator1
    Position Offset=<0 0 5>
    Sprite Rotation Mode=::NewEnumerator2
    Sprite Size Mode=::NewEnumerator3
    Uniform Sprite Size=$Particles.SizeCurve
    UsePositionOffset=true
    Write Material Random=true
   )
   &ShapeLocation(
    MaskSpawnGroup=false
    Offset Coordinate Space=::World
    Offset Mode=::NewEnumerator0
    Ring / Disc Distribution Mode:ENiagaraSphereDistributionMode=::NewEnumerator3
    Ring Radius=10
    Shape Primitive=::NewEnumerator4
    UseOwnerScale=true
   )
   &AddVelocity(Velocity=&Multiply_VectorByFloat(Vector=$Particles.ShapeLocation.ShapeTangent, Float=&RandomRangeFloat(Minimum=80, Maximum=120))) {module=AddVelocity001}
   &AddVelocity(Coordinate Space=::World, Velocity=&RandomRangeVector(Minimum=<0 0 25>, Maximum=<0 0 30>))
   &CameraOffset(Camera Offset Amount=10)
  ]
  particleUpdateScript [
   &ParticleState()
   &Drag(Drag=4, Use Linear Drag=true, Use Rotational Drag=false)
   &GravityForce(Gravity=<0 0 -50>)
   &SolveForcesAndVelocity(Clamp Velocity=false, Limit Acceleration=false)
   &SpriteRotationRate(Rotation Rate=&Multiply_Float(
     A=$Particles.InitSpriteRotRate
     B=&FloatFromCurve(
      FloatCurve=DataInterface() {propertyValues {Curve {keys [
          {arriveTangent=-3.9021, interpMode=RCIM_Cubic, leaveTangent=-3.9021, leaveTangentWeight=1.3778, tangentMode=RCTM_User, time=0, value=1}
          {arriveTangentWeight=0.3436, interpMode=RCIM_Cubic, tangentMode=RCTM_User, time=0.75, value=0.1}
         ]}}}
      Scale Curve=50
     )
    ))
   &ScaleSpriteSize(
    Scale Sprite Size Mode=::NewEnumerator0
    Uniform Scale Factor=&FloatFromCurve(FloatCurve=DataInterface() {propertyValues {Curve {keys [
         {interpMode=RCIM_Cubic, leaveTangent=1.0601, tangentMode=RCTM_Break, time=0, value=0.2}
         {arriveTangent=0.5926, interpMode=RCIM_Cubic, tangentMode=RCTM_Break, time=0.95, value=1}
         <at 1 0>
        ]}}})
   )
   &DynamicMaterialParameters(
    Index 0 Param 1=&Add_Float(
     A=&RandomRangeFloat(Minimum=0, Maximum=1)
     B=&FloatFromCurve(
      FloatCurve=DataInterface() {propertyValues {Curve {keys [{tangentMode=RCTM_User, time=0, value=0}, {tangentMode=RCTM_User, time=1, value=1}]}}}
      CurveIndex=&Add_Float(A=&Multiply_Float(A=$Particles.Initial.MaterialRandom, B=1), B=$Particles.NormalizedAge)
      Scale Curve=0.1
     )
    )
    Index 0 Param 2=&FloatFromCurve(FloatCurve=DataInterface() {propertyValues {Curve {keys [<at 0 1>,<at 1 0>]}}})
    Index 0 Param 3=&FloatFromCurve(
     FloatCurve=DataInterface() {propertyValues {Curve {keys [
         {interpMode=RCIM_Cubic, leaveTangent=16.5652, leaveTangentWeight=0.4385, tangentMode=RCTM_Break, time=0, value=0}
         {arriveTangentWeight=0.0333, interpMode=RCIM_Cubic, leaveTangentWeight=0.9477, tangentMode=RCTM_Break, time=0.1, value=1}
         {arriveTangent=-1.7159, interpMode=RCIM_Cubic, leaveTangent=-1.7159, tangentMode=RCTM_User, time=0.3, value=0.2}
         {arriveTangent=0.2347, interpMode=RCIM_Cubic, leaveTangent=0.2723, tangentMode=RCTM_Break, time=0.6, value=0}
        ]}}}
     Scale Curve=0
    )
    Index 0 Param 4=&FloatFromCurve(FloatCurve=DataInterface() {propertyValues {Curve {keys [{interpMode=RCIM_Cubic, tangentMode=RCTM_Break, time=0.5, value=1}, {interpMode=RCIM_Cubic, tangentMode=RCTM_Break, time=1, value=0}]}}})
    Index 1 Param 1=&FloatFromCurve(FloatCurve=DataInterface() {propertyValues {Curve {keys [
         {interpMode=RCIM_Cubic, leaveTangentWeight=0.0833, tangentMode=RCTM_Break, time=0, value=0}
         {arriveTangentWeight=0.0873, interpMode=RCIM_Cubic, tangentMode=RCTM_Break, time=0.25, value=1}
        ]}}})
    Index 1 Param 2=&FloatFromCurve(
     FloatCurve=DataInterface() {propertyValues {Curve {keys [
         {interpMode=RCIM_Cubic, tangentMode=RCTM_Break, time=0, value=0.2}
         {interpMode=RCIM_Cubic, leaveTangent=0.9066, tangentMode=RCTM_Break, time=1, value=2}
        ]}}}
     Scale Curve=&RandomRangeFloat(Minimum=0.1, Maximum=0.2)
    )
    Index 1 Param 3=&FloatFromCurve(FloatCurve=DataInterface() {propertyValues {Curve {keys [
         {interpMode=RCIM_Cubic, leaveTangent=6.3917, tangentMode=RCTM_Break, time=0, value=0}
         {arriveTangent=1.7549, interpMode=RCIM_Cubic, leaveTangent=10.6767, tangentMode=RCTM_User, time=0.25, value=1}
        ]}}})
    Index 1 Param 4=&FloatFromCurve(
     FloatCurve=DataInterface() {propertyValues {Curve {keys [{arriveTangent=-0.136, interpMode=RCIM_Cubic, leaveTangent=-0.136, tangentMode=RCTM_User, time=0, value=1}, <at 1 0.95>]}}}
     Scale Curve=2
    )
    Index 2 Param 1=&FloatFromCurve(
     FloatCurve=DataInterface() {propertyValues {Curve {keys [
         {interpMode=RCIM_Cubic, leaveTangent=4.2485, tangentMode=RCTM_Break, time=0, value=0}
         {arriveTangent=1.0548, interpMode=RCIM_Cubic, leaveTangent=1.0548, tangentMode=RCTM_User, time=0.17, value=0.5715}
         <at 1 1>
        ]}}}
     Scale Curve=3
    )
    Index 2 Param 2=&FloatFromCurve(
     FloatCurve=DataInterface() {propertyValues {Curve {keys [{interpMode=RCIM_Cubic, tangentMode=RCTM_User, time=0, value=0}, {interpMode=RCIM_Cubic, tangentMode=RCTM_User, time=1, value=1}]}}}
     Scale Curve=0.3
    )
    Index 2 Param 3=&FloatFromCurve(
     FloatCurve=DataInterface() {propertyValues {Curve {keys [
         {interpMode=RCIM_Cubic, leaveTangent=-15.7052, leaveTangentWeight=0.1411, tangentMode=RCTM_Break, time=0, value=1}
         {interpMode=RCIM_Cubic, tangentMode=RCTM_Break, time=0.1, value=0}
         {arriveTangent=0.7532, interpMode=RCIM_Cubic, leaveTangent=0.7532, tangentMode=RCTM_User, time=0.7, value=0.5}
         {arriveTangent=2.358, arriveTangentWeight=1.057, interpMode=RCIM_Cubic, tangentMode=RCTM_Break, time=1, value=1}
        ]}}}
     Scale Curve=5
    )
    Index 2 Param 4=&RandomRangeFloat(Minimum=1.5, Maximum=1.75)
    Module.Param0WriteEnabled=true
    Module.Param0WriteEnabled1=true
    Module.Param0WriteEnabled2=true
    Module.Param1WriteEnabled=false
    Module.Param1WriteEnabled1=true
    Module.Param1WriteEnabled2=true
    Module.Param2WriteEnabled=true
    Module.Param2WriteEnabled1=true
    Module.Param2WriteEnabled2=true
    Module.Param3WriteEnabled=true
    Module.Param3WriteEnabled1=true
    Module.Param3WriteEnabled2=true
    Write Parameter Index 1=true
    Write Parameter Index 2=true
   )
   &ScaleColor(
    Scale Alpha=&FloatFromCurve(
     FloatCurve=DataInterface() {propertyValues {Curve {keys [
         {arriveTangent=6.442, interpMode=RCIM_Cubic, leaveTangent=6.442, tangentMode=RCTM_User, time=0, value=0}
         {arriveTangent=-1.0806, interpMode=RCIM_Cubic, leaveTangent=-1.0806, tangentMode=RCTM_User, time=0.03, value=1}
         <at 1 0>
        ]}}}
     Scale Curve=&RandomRangeFloat(Minimum=0.45, Maximum=0.6)
    )
    ScaleA=true
    ScaleRGB=false
   )
  ]
  rendererClasses [&NiagaraSpriteRendererProperties]
  renderers [
   {rendererClass=&NiagaraSpriteRendererProperties}
   {rendererClass=&NiagaraSpriteRendererProperties, rendererIndex=1}
  ]
 }
 Shockwave {
  emitterUpdateScript [
   &EmitterState()
   &SpawnBurst_Instantaneous(Spawn Group=1, Use Spawn Probability=false, UseLoopCountLimit=false) {module=SpawnBurst_Instantaneous001}
  ]
  particleSpawnScript [
   &InitializeParticle(
    Color=&FNTimeOfDayColor(Night Color=<rgba 1 1 1 0.175>, Morning Color=<rgba 1 1 1 0.175>, Day Color=<rgba 1 1 1 0.075>, Evening Color=<rgba 1 1 1 0.125>)
    Lifetime=0.4
    Mesh Scale Mode=::NewEnumerator3
    Mesh Uniform Scale=0.05
    Position Offset Coordinate Space=::World
    UsePositionOffset=true
    Write Material Random=true
   )
   &InitialMeshOrientation(Mesh Orientation Mode=::NewEnumerator2, Rotation=<0 0 0.2>, Use Rotation Vector=true)
  ]
  particleUpdateScript [
   &ParticleState()
   &UpdateMeshOrientation(Rotation Vector=<0 0 1>)
   &DynamicMaterialParameters(
    Index 0 Param 1=&FloatFromCurve(FloatCurve=DataInterface() {propertyValues {Curve {keys [{interpMode=RCIM_Cubic, tangentMode=RCTM_User, time=0.5, value=0}, {interpMode=RCIM_Cubic, tangentMode=RCTM_User, time=1, value=1}]}}})
    Index 0 Param 3=1
    Module.Param0WriteEnabled=true
    Module.Param1WriteEnabled=true
    Module.Param2WriteEnabled=true
    Module.Param3WriteEnabled=true
   )
   &ScaleMeshSize(Scale Factor=&VectorFromCurve(
     VectorCurve=DataInterface() {propertyValues {
       XCurve {keys [
         {arriveTangent=4.1785, interpMode=RCIM_Cubic, leaveTangent=4.1785, tangentMode=RCTM_User, time=0, value=0}
         {arriveTangent=1.1179, interpMode=RCIM_Cubic, leaveTangent=1.1179, tangentMode=RCTM_User, time=0.2042, value=0.6228}
         {arriveTangent=0.0224, interpMode=RCIM_Cubic, leaveTangent=0.0224, tangentMode=RCTM_User, time=1, value=1}
        ]}
       YCurve {keys [
         {arriveTangent=4.1785, interpMode=RCIM_Cubic, leaveTangent=4.1785, tangentMode=RCTM_User, time=0, value=0}
         {arriveTangent=1.1179, interpMode=RCIM_Cubic, leaveTangent=1.1179, tangentMode=RCTM_User, time=0.2042, value=0.6228}
         {arriveTangent=0.0224, interpMode=RCIM_Cubic, leaveTangent=0.0224, tangentMode=RCTM_User, time=1, value=1}
        ]}
       ZCurve {keys [{tangentMode=RCTM_User, time=0, value=1}, {tangentMode=RCTM_User, time=1, value=1}]}
      }}
     Scale Curve=<5 5 5>
    ))
   &ScaleColor(
    Scale Alpha=&FloatFromCurve(
     FloatCurve=DataInterface() {propertyValues {Curve {keys [
         {arriveTangent=0.0617, interpMode=RCIM_Cubic, leaveTangent=0.0617, tangentMode=RCTM_User, time=0.1, value=1}
         {arriveTangent=0.0658, interpMode=RCIM_Cubic, leaveTangent=0.0658, tangentMode=RCTM_User, time=0.7, value=1}
         {tangentMode=RCTM_User, time=1, value=0}
        ]}}}
     Scale Curve=&Multiply_Float(A=1, B=1)
    )
    Scale RGB=&MakeVectorFromLinearColorRGB(Color=&MultiplyLinearColor(
      Linear Color=&RandomRangeLinearColor(Minimum=<rgba 0.5 0.5 0.5 1>, Maximum=<rgba 1 1 1 1>)
      Scale Factor=&MakeVector4_FromLinearColor(LinearColor=&LerpLinearColors(StartColor=<rgba 1 1 1 1>, EndColor=<rgba 0 0 0 1>, LerpFactor=0.1))
     ))
    ScaleA=true
    ScaleRGB=true
   ) {module=ScaleColor001}
   &ScaleColor(
    Scale Alpha=&FloatFromCurve(
     FloatCurve=DataInterface() {propertyValues {Curve {keys [
         {arriveTangent=6.442, interpMode=RCIM_Cubic, leaveTangent=6.442, tangentMode=RCTM_User, time=0, value=0}
         {interpMode=RCIM_Cubic, tangentMode=RCTM_User, time=0.1, value=1}
         <at 1 0>
        ]}}}
     Scale Curve=&RandomRangeFloat(Minimum=0.75, Maximum=1)
    )
    ScaleA=true
    ScaleRGB=false
   ) {enabled=false}
  ]
  rendererClasses [&NiagaraMeshRendererProperties]
  renderers [{rendererClass=&NiagaraMeshRendererProperties, rendererIndex=0}]
 }
}
summary {
 systemName=NS_Emote_CiderPetalClover_ReturnImpactDust
 emitters [
  {emitterName=Dust, rendererClasses [&NiagaraSpriteRendererProperties]}
  {emitterName=Shockwave, rendererClasses [&NiagaraMeshRendererProperties]}
 ]
 userVariables [
  User.bIsFrontEnd {type=NiagaraBool}
  User.bIsFrontEndPreview {type=NiagaraBool}
  User.BurstIndex {type=NiagaraFloat, defaultValue=0, description=""}
  User.DesertBoulder {type=NiagaraBool}
  User.FrontEndScale {type=NiagaraFloat, defaultValue=0.4, description=""}
 ]
}
