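# Niagara system for a short-lived dark blade trail in the CiderPetalClover player emote, named as the trail's end effect.
# A single emitter spawns black anim-trail particles from skeletal mesh bones or sockets, drawn by two ribbon renderers.
# Curves indexed by system age narrow the ribbon and fade it out within about 0.2 seconds, combined with a per-particle alpha fade.
# A similar fade drives the alpha of a white HighLight particle attribute.
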NiagaraSystem /BRCosmetics/Animation/Game/MainPlayer/Emotes/CiderPetalClover/FX/NS_Emote_CiderPetalClover_TrailEnd
dependencies {
 usedDataInterfaces [&NiagaraDataInterfaceCurve, &NiagaraDataInterfaceSkeletalMesh]
 usedDynamicInputs [&FN_Select_Float_from_Time_Of_Day]
 usedModules [
  &self:SystemSpawnScript.SystemScriptSource.SystemScriptGraph.NiagaraNodeAssignment_0.SetVariables_#1
  &SystemState
 ]
}
emitters {BladeTrailDark {
  emitterUpdateScript [
   &EmitterState()
   &FN_SpawnAnimTrail(
    Spawn Interval=0
    Spawn Mode:ENiagaraSkeletalSpawning=::NewEnumerator0
    Switch on ENiagara_AnimTrail_Source:ENiagara_AnimTrail_Source=::NewEnumerator1
   )
  ]
  particleSpawnScript [
   &FN_InitializeAnimTrailBoneIndecies(Second Bone or Socket Index=0)
   &FN_InitializeAnimTrail(
    AnimTrailWidth=&Multiply_Float(A=27, B=1)
    Color=&MultiplyLinearColor(Linear Color=<rgba 0 0 0 2>, Scale Factor=<1 1 1 1>)
    Lifetime=0.25
    SkeletalMesh=$System.Skeletal Mesh
   )
  ]
  particleUpdateScript [
   &FN_UpdateAnimTrail(SkeletalMesh=$System.Skeletal Mesh) {module=FN_UpdateAnimTrail001}
   &ScaleColor(
    Scale Alpha=&Multiply_Float(
     A=&Multiply_Float(
      A=&FloatFromCurve(FloatCurve=DataInterface() {propertyValues {Curve {keys [<at -0.0046 1.0113 RCIM_Cubic>,<at 1.0011 0.002 RCIM_Cubic>]}}})
      B=&FloatFromCurve(
       FloatCurve=DataInterface() {propertyValues {Curve {
          defaultValue=1
          keys [
           {tangentMode=RCTM_User, time=0, value=1}
           {arriveTangent=-0.0405, interpMode=RCIM_Cubic, leaveTangent=-0.0405, tangentMode=RCTM_User, time=0.1, value=1}
           {arriveTangent=-8.0089, interpMode=RCIM_Cubic, leaveTangent=-8.0089, tangentMode=RCTM_User, time=0.2, value=0}
          ]
         }}}
       CurveIndex=$System.Age
      )
     )
     B=0.7
    )
    ScaleA=true
    ScaleRGB=false
   )
   &ScaleRibbonWidth(Ribbon Width Scale=&FloatFromCurve(
     FloatCurve=DataInterface() {propertyValues {Curve {
        defaultValue=1
        keys [
         {tangentMode=RCTM_User, time=0, value=1}
         {arriveTangent=-0.0405, interpMode=RCIM_Cubic, leaveTangent=-0.0405, tangentMode=RCTM_User, time=0.075, value=1}
         {arriveTangent=-8.0089, interpMode=RCIM_Cubic, leaveTangent=-8.0089, tangentMode=RCTM_User, time=0.2, value=0.4}
        ]
       }}}
     CurveIndex=$System.Age
    ))
   /SetVariables(Particles.HighLight:LinearColor=&MakeLinearColorFromVectorAndFloat(
     "Vector (RGB)"=<1 1 1>
     "Float (Alpha)"=&Multiply_Float(
      A=2
      B=&Multiply_Float(
       A=&Multiply_Float(
        A=&FloatFromCurve(FloatCurve=DataInterface() {propertyValues {Curve {keys [<at -0.0046 1.0113 RCIM_Cubic>,<at 1.0011 0.002 RCIM_Cubic>]}}})
        B=&FloatFromCurve(
         FloatCurve=DataInterface() {propertyValues {Curve {
            defaultValue=1
            keys [
             {tangentMode=RCTM_User, time=0, value=1}
             {arriveTangent=-0.0405, interpMode=RCIM_Cubic, leaveTangent=-0.0405, tangentMode=RCTM_User, time=0.1, value=1}
             {arriveTangent=-8.0089, interpMode=RCIM_Cubic, leaveTangent=-8.0089, tangentMode=RCTM_User, time=0.19, value=0}
            ]
           }}}
         CurveIndex=$System.Age
        )
       )
       B=1
      )
     )
    )) {isSetParameters=true}
  ]
  rendererClasses [&NiagaraRibbonRendererProperties]
  renderers [
   {rendererClass=&NiagaraRibbonRendererProperties}
   {rendererClass=&NiagaraRibbonRendererProperties, rendererIndex=1}
  ]
 }}
summary {
 systemName=NS_Emote_CiderPetalClover_TrailEnd
 emitters [{emitterName=BladeTrailDark, rendererClasses [&NiagaraRibbonRendererProperties]}]
 userVariables [
  User.bIsFrontEnd {type=NiagaraBool}
  User.bIsFrontEndPreview {type=NiagaraBool}
  User.FrontEndScale {type=NiagaraFloat, defaultValue=0.6, description=""}
  User.RibbonIDOffset {type=NiagaraInt32}
  User.SwingSectionIndex {type=NiagaraInt32}
  User.Variant {type=NiagaraFloat, defaultValue=0, description=""}
 ]
}
