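# Niagara system for the trail-start effect of the CiderPetalClover emote, drawing short-lived streaks from the character's skeletal mesh.
# BladeTrailDark builds black anim-trail ribbons from bones or sockets, fading and narrowing quickly with system age and driving a white highlight parameter.
# Core briefly spawns mesh-sampled particles, offset along their velocity, with sprite and ribbon renderers and an eroding, distorted grey ribbon material.
# Core particle lifetime steps down by platform quality, falling back to zero.
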NiagaraSystem /BRCosmetics/Animation/Game/MainPlayer/Emotes/CiderPetalClover/FX/NS_Emote_CiderPetalClover_TrailStart
dependencies {
 usedDataInterfaces [
  &NiagaraDataInterfacePlatformSet
  &NiagaraDataInterfaceColorCurve
  &NiagaraDataInterfaceCurve
  &NiagaraDataInterfaceSkeletalMesh
 ]
 usedDynamicInputs [&FN_Select_Float_from_Time_Of_Day]
 usedModules [
  &self:SystemSpawnScript.SystemScriptSource.SystemScriptGraph.NiagaraNodeAssignment_0.SetVariables_#1
  &SystemState
 ]
}
emitters {
 BladeTrailDark {
  emitterUpdateScript [
   &EmitterState()
   &FN_SpawnAnimTrail(
    Spawn Interval=0
    Spawn Mode:ENiagaraSkeletalSpawning=::NewEnumerator0
    Switch on ENiagara_AnimTrail_Source:ENiagara_AnimTrail_Source=::NewEnumerator1
   )
  ]
  particleSpawnScript [
   &FN_InitializeAnimTrailBoneIndecies(Second Bone or Socket Index=0)
   &FN_InitializeAnimTrail(
    AnimTrailWidth=&Multiply_Float(A=27, B=1)
    Color=&MultiplyLinearColor(Linear Color=<rgba 0 0 0 2>, Scale Factor=<1 1 1 1>)
    Lifetime=0.25
    SkeletalMesh=$System.Skeletal Mesh
   )
  ]
  particleUpdateScript [
   &FN_UpdateAnimTrail(SkeletalMesh=$System.Skeletal Mesh) {module=FN_UpdateAnimTrail001}
   &ScaleColor(
    Scale Alpha=&Multiply_Float(
     A=&Multiply_Float(
      A=&FloatFromCurve(FloatCurve=DataInterface() {propertyValues {Curve {keys [<at -0.0046 1.0113 RCIM_Cubic>,<at 1.0011 0.002 RCIM_Cubic>]}}})
      B=&FloatFromCurve(
       FloatCurve=DataInterface() {propertyValues {Curve {
          defaultValue=1
          keys [
           {tangentMode=RCTM_User, time=0, value=1}
           {arriveTangent=-0.0405, interpMode=RCIM_Cubic, leaveTangent=-0.0405, tangentMode=RCTM_User, time=0.1, value=1}
           {arriveTangent=-8.0089, interpMode=RCIM_Cubic, leaveTangent=-8.0089, tangentMode=RCTM_User, time=0.2, value=0}
          ]
         }}}
       CurveIndex=$System.Age
      )
     )
     B=0.8
    )
    ScaleA=true
    ScaleRGB=false
   )
   &ScaleRibbonWidth(Ribbon Width Scale=&FloatFromCurve(
     FloatCurve=DataInterface() {propertyValues {Curve {
        defaultValue=1
        keys [
         {tangentMode=RCTM_User, time=0, value=1}
         {arriveTangent=-0.0405, interpMode=RCIM_Cubic, leaveTangent=-0.0405, tangentMode=RCTM_User, time=0.075, value=1}
         {arriveTangent=-8.0089, interpMode=RCIM_Cubic, leaveTangent=-8.0089, tangentMode=RCTM_User, time=0.2, value=0.4}
        ]
       }}}
     CurveIndex=$System.Age
    ))
   /SetVariables(Particles.HighLight:LinearColor=&MakeLinearColorFromVectorAndFloat(
     "Vector (RGB)"=<1 1 1>
     "Float (Alpha)"=&Multiply_Float(
      A=2
      B=&Multiply_Float(
       A=&Multiply_Float(
        A=&FloatFromCurve(FloatCurve=DataInterface() {propertyValues {Curve {keys [<at -0.0046 1.0113 RCIM_Cubic>,<at 1.0011 0.002 RCIM_Cubic>]}}})
        B=&FloatFromCurve(
         FloatCurve=DataInterface() {propertyValues {Curve {
            defaultValue=1
            keys [
             {tangentMode=RCTM_User, time=0, value=1}
             {arriveTangent=-0.0405, interpMode=RCIM_Cubic, leaveTangent=-0.0405, tangentMode=RCTM_User, time=0.1, value=1}
             {arriveTangent=-8.0089, interpMode=RCIM_Cubic, leaveTangent=-8.0089, tangentMode=RCTM_User, time=0.19, value=0}
            ]
           }}}
         CurveIndex=$System.Age
        )
       )
       B=1
      )
     )
    )) {isSetParameters=true}
  ]
  rendererClasses [&NiagaraRibbonRendererProperties]
  renderers [
   {rendererClass=&NiagaraRibbonRendererProperties}
   {rendererClass=&NiagaraRibbonRendererProperties, rendererIndex=1}
  ]
 }
 Core {
  emitterUpdateScript [
   &EmitterState(Life Cycle Mode=::NewEnumerator1, Loop Behavior=::NewEnumerator1, Loop Duration=0.05, UseLoopDelay=false)
   &SpawnRate(SpawnRate=120, Use Spawn Probability=false)
  ]
  particleSpawnScript [
   &InitializeParticle(
    Lifetime=&FN_PickFloatDependOnPlatform(
     TargetPlatform=DataInterface() {propertyValues {Platforms {qualityLevelMask=-8}}}
     TargetPlatformActive=0.2
     TargetPlatformNoActive=&FN_PickFloatDependOnPlatform(
      TargetPlatform=DataInterface() {propertyValues {Platforms {qualityLevelMask=-28}}}
      TargetPlatformActive=0.18
      TargetPlatformNoActive=&FN_PickFloatDependOnPlatform(
       TargetPlatform=DataInterface() {propertyValues {Platforms {qualityLevelMask=-30}}}
       TargetPlatformActive=0.16
       TargetPlatformNoActive=&FN_PickFloatDependOnPlatform(TargetPlatform=DataInterface() {propertyValues {Platforms {qualityLevelMask=-31}}}, TargetPlatformActive=0.14, TargetPlatformNoActive=0)
      )
     )
    )
    Mass Mode=::NewEnumerator3
    Ribbon Width=35
    Ribbon Width Mode=::NewEnumerator1
    Sprite Rotation Mode=::NewEnumerator2
    Sprite Size Mode=::NewEnumerator3
    UsePositionOffset=false
    Write Material Random=true
   )
   &SkeletalMeshLocation(
    Bone Sampling Mode=::NewEnumerator4
    Inherited Velocity Speed Limit=1
    MaskSpawnGroup=false
    Skeletal Mesh=$System.Skeletal Mesh
    Velocity Sampling=::NewEnumerator0
    VelocitySpeedLimit=true
   ) {module=SkeletalMeshLocation002}
   &AddVelocity(
    Cone Angle=32
    Cone Axis=<0 0 1>
    Distribution Along Cone Axis=0.25
    Module.UseConeFalloff=true
    Module.UseVelDistribution=true
    Override Seed=false
    Speed Falloff From Cone Axis=0.5
    Velocity Mode=::NewEnumerator2
    Velocity Speed=&RandomRangeFloat(Minimum=500, Maximum=850)
   ) {enabled=false}
   &OffsetPosition(
    Position Offset=&Multiply_VectorByFloat(Vector=$Particles.Velocity, Float=10)
    Position Offset Coordinate Space:ENiagaraCoordinateSpace=::World
   )
  ]
  particleUpdateScript [
   &ParticleState()
   &GravityForce() {enabled=false}
   &FN_RibbonMaterial(
    Color=&ColorFromCurve(ColorCurve=DataInterface() {propertyValues {
       AlphaCurve {keys [<at 0 2 RCIM_Cubic>,<at 0.0513 2>]}
       BlueCurve {keys [<at 0.0258 0.5089>]}
       GreenCurve {keys [<at 0.0258 0.4793>]}
       RedCurve {keys [<at 0.0258 0.4735>]}
      }})
    Erosion Amount:NiagaraFloat=&Lerp_Float(A=0.25, B=1, Alpha=$Particles.NormalizedAge)
    Erossion Scale U=4.5
    Macro Distortion Amount=0.1
    Macro Pan Offset=-0.1216
    Micro Distortion Amount=0.0384
    Opacity Erosion:NiagaraFloat=&FloatFromCurve(FloatCurve=DataInterface() {propertyValues {Curve {keys [<at 0 0 RCIM_Cubic>, {arriveTangent=1.1077, tangentMode=RCTM_User, time=1, value=1}]}}})
    Opacity Power:NiagaraFloat=&FloatFromCurve(
     FloatCurve=DataInterface() {propertyValues {Curve {keys [<at 0 0 RCIM_Cubic>, {arriveTangent=1.1077, tangentMode=RCTM_User, time=1, value=1}]}}}
     CurveIndex=$Emitter.NormalizedLoopAge
    )
    Scale Alpha:NiagaraFloat=&FloatFromCurve(FloatCurve=DataInterface() {propertyValues {Curve {keys [
         <at 0 0.7 RCIM_Cubic>
         {arriveTangent=-0.6429, interpMode=RCIM_Cubic, leaveTangent=-0.6429, time=0.2, value=0.45}
         {arriveTangent=-0.5592, interpMode=RCIM_Cubic, leaveTangent=-0.5592, tangentMode=RCTM_User, time=0.7, value=0.25}
         {arriveTangent=-0.4965, tangentMode=RCTM_User, time=1, value=0}
        ]}}})
    UV Basis:NiagaraFloat=&ReturnNormalizedExecIndex()
   ) {module=FN_RibbonMaterial001}
   &Drag(Drag=0.25, Use Linear Drag=true, Use Rotational Drag=false)
   &ScaleColor(
    Scale Alpha=&Multiply_Float(
     A=&FloatFromCurve(FloatCurve=DataInterface() {propertyValues {Curve {keys [<at -0.0046 1.0113 RCIM_Cubic>,<at 1.0011 0.002 RCIM_Cubic>]}}}, Scale Curve=0.7)
     B=&FloatFromCurve(
      FloatCurve=DataInterface() {propertyValues {Curve {keys [<at -0.0046 1.0113 RCIM_Cubic>,<at 1.0011 0.002 RCIM_Cubic>]}}}
      CurveIndex=$Emitter.NormalizedLoopAge
     )
    )
    ScaleA=true
    ScaleRGB=true
   )
   &SolveForcesAndVelocity(Clamp Velocity=false, Limit Acceleration=false)
  ]
  rendererClasses [&NiagaraSpriteRendererProperties, &NiagaraRibbonRendererProperties]
  renderers [
   {rendererClass=&NiagaraSpriteRendererProperties}
   {rendererClass=&NiagaraRibbonRendererProperties, rendererIndex=1}
  ]
 }
}
summary {
 systemName=NS_Emote_CiderPetalClover_TrailStart
 emitters [
  {emitterName=Core, rendererClasses [&NiagaraSpriteRendererProperties, &NiagaraRibbonRendererProperties]}
  {emitterName=BladeTrailDark, rendererClasses [&NiagaraRibbonRendererProperties]}
 ]
 userVariables [
  User.bIsFrontEnd {type=NiagaraBool}
  User.bIsFrontEndPreview {type=NiagaraBool}
  User.FrontEndScale {type=NiagaraFloat, defaultValue=0.6, description=""}
  User.RibbonIDOffset {type=NiagaraInt32}
  User.SwingSectionIndex {type=NiagaraInt32}
  User.Variant {type=NiagaraFloat, defaultValue=0, description=""}
 ]
}
